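# Niagara system for the DistanceHappening cosmetic skydiving contrail, layering glow cores, dots, galaxy smoke, mesh swirls, hand trails and vortex sparks.
# Emitters split particles by partition across sampled skeletal-mesh bones or sockets, with spawn groups selecting per-layer values and visibility tags.
# Glide lean drives rates, sizes, brightness and forces, while dive feeds swirl material parameters.
# Time of day and front-end state adjust the look; platform and quality checks scale smoke opacity.
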NiagaraSystem /BRCosmetics/Athena/Effects/Particles/Skydiving/NS_Contrail_DistanceHappening/NS_Contrail_DistanceHappening
dependencies {
 usedDataInterfaces [
  &NiagaraDataInterfaceCurve
  &NiagaraDataInterfaceColorCurve
  &NiagaraDataInterfaceSpriteRendererInfo
  &NiagaraDataInterfaceVectorCurve
  &NiagaraDataInterfacePlatformSet
  &NiagaraDataInterfaceArrayFloat3
  &NiagaraDataInterfaceSkeletalMesh
 ]
 usedDynamicInputs []
 usedModules [
  &self:SystemSpawnScript.SystemScriptSource.SystemScriptGraph.NiagaraNodeAssignment_0.SetVariables_#1
  &self:SystemSpawnScript.SystemScriptSource.SystemScriptGraph.NiagaraNodeAssignment_2.SetVariables_#2
  &self:SystemSpawnScript.SystemScriptSource.SystemScriptGraph.NiagaraNodeAssignment_3.SetVariables_#3
  &SystemState
  /Game/Effects/Niagara/Modules/Contrails/FN_Contrail_Setup
 ]
}
emitters {
 Core_Glow_Refraction {
  emitterUpdateScript [
   &EmitterState(Life Cycle Mode=::NewEnumerator1, Loop Behavior=::NewEnumerator1, Loop Duration Mode=::NewEnumerator1, UseLoopDelay=false)
   &SpawnBurst_Instantaneous(Spawn Count=2, Use Spawn Probability=false, UseLoopCountLimit=false)
   &SpawnBurst_Instantaneous(Spawn Count=2, Spawn Group=1, Use Spawn Probability=false, UseLoopCountLimit=false) {module=SpawnBurst_Instantaneous001}
   &SpawnBurst_Instantaneous(Spawn Count=2, Spawn Group=2, Use Spawn Probability=false, UseLoopCountLimit=false) {module=SpawnBurst_Instantaneous002}
   &SpawnBurst_Instantaneous(Spawn Count=2, Spawn Group=3, Use Spawn Probability=false, UseLoopCountLimit=false) {module=SpawnBurst_Instantaneous003}
   &SpawnBurst_Instantaneous(Spawn Count=2, Spawn Group=4, Use Spawn Probability=false, UseLoopCountLimit=false) {module=SpawnBurst_Instantaneous004}
  ]
  particleSpawnScript [
   /SetVariables(Particles.VisibilityTag=&SpawnGroup()) {isSetParameters=true}
   &PartitionParticles()
   &InitializeParticle(
    Color=&MergeHelper_ColorSwitch(
     0=<rgba 0.0008 0.0007 0.0321 1>
     1=&FNTimeOfDayColor(Night Color=<rgba 1.2 1.2 1.2 0.6>, Morning Color=<rgba 1 1 1 0.6>, Day Color=<rgba 1 1 1 0.6>, Evening Color=<rgba 1.2 1.2 1.2 0.6>)
     2=&FNTimeOfDayColor(
      Night Color=<rgba 3.89 3.89 10 3>
      Morning Color=<rgba 1.9427 1.9427 5 2.5>
      Day Color=<rgba 1.9427 1.9427 5 2.5>
      Evening Color=<rgba 3.89 3.89 10 3>
     )
     3=&FNTimeOfDayColor(
      Night Color=<rgba 0.3409 0.7367 3 0.45>
      Morning Color=<rgba 1.0734 0.7138 3 0.2>
      Day Color=<rgba 0.3409 0.7367 3 0.2>
      Evening Color=<rgba 0.3409 0.7367 3 0.45>
     )
     4=&FNTimeOfDayColor(Night Color=<rgba 1 1 1 20>, Morning Color=<rgba 1 1 1 5>, Day Color=<rgba 1 1 1 5>, Evening Color=<rgba 1 1 1 20>)
     Index=&SpawnGroup()
     Num=::NewEnumerator3
    )
    Lifetime=&MergeHelper_FloatSwitch(0=3, 1=3, 2=3, 3=3, 4=1, Index=&SpawnGroup(), Num=::NewEnumerator3)
    Mesh Scale Mode=::NewEnumerator3
    Mesh Uniform Scale=&MergeHelper_FloatSwitch(
     0=0
     1=0
     2=0
     3=0.17
     4=&Lerp_Float(A=0.13, B=0.16, Alpha=$User.ParaGlideLeanAlpha)
     Index=&SpawnGroup()
     Num=::NewEnumerator3
    )
    Sprite Size Mode=::NewEnumerator3
    Uniform Sprite Size=&MergeHelper_FloatSwitch(0=15, 1=45, 2=20, 3=0, Index=&SpawnGroup(), Num=::NewEnumerator2)
    UsePositionOffset=false
    Write Material Random=true
   )
   &InitialMeshOrientation(
    Mesh Orientation Mode=::NewEnumerator1
    Rotation=&MergeHelper_VectorSwitch(
     0=<0 0 0>
     1=<0 0 0>
     2=<0 0 0>
     3=&FN_FrontEndVector3Modifier(Gameplay=<0.2 0.2 -0.15>, Front End=<0.34 0 -0.1>)
     4=&FN_FrontEndVector3Modifier(Gameplay=<0.2 0.2 -0.15>, Front End=<0.34 0 -0.1>)
     Index=&SpawnGroup()
     Num=::NewEnumerator3
    )
    Rotation Coordinate Space=::NewEnumerator2
    Use Rotation Vector=true
   )
   &CameraOffset(Camera Offset Amount=&MergeHelper_FloatSwitch(0=30, 1=30, 2=0, 3=25, Index=&SpawnGroup(), Num=::NewEnumerator2))
   /SetVariables(Particles.DMP_HollowGlow:Vector4f=&MakeVector4(X=2, Y=10, Z=0, W=0)) {isSetParameters=true}
  ]
  particleUpdateScript [
   &SkeletalMeshLocation(
    Bone / Socket Index=$Particles.PartitionParticles.Partition
    Bone Sampling Mode=::NewEnumerator4
    MaskSpawnGroup=false
    Sampled Position Offset=&MergeHelper_VectorSwitch(0=<10 10 0>, 1=<-10 -10 0>, 2=?, 3=?, Index=$Particles.PartitionParticles.Partition, Num=::NewEnumerator0)
    Skeletal Mesh=$System.Skeletal Mesh Reader
   )
   &ParticleState(Kill Particles When Lifetime Has Elapsed=false)
   &SpriteRotationRate(Rotation Rate=&MergeHelper_FloatSwitch(0=-80, 1=80, 2=?, 3=?, Index=$Particles.Initial.PartitionParticles.Partition, Num=::NewEnumerator0))
   &ScaleSpriteSize(
    Scale Sprite Size Mode=::NewEnumerator0
    Uniform Scale Factor=&MergeHelper_FloatSwitch(
     0=&FloatFromCurve(
      FloatCurve=DataInterface() {propertyValues {Curve {
         defaultValue=1
         keys [{interpMode=RCIM_Cubic, tangentMode=RCTM_User, time=0, value=0}, {interpMode=RCIM_Cubic, tangentMode=RCTM_User, time=0.3, value=1}]
        }}}
      Scale Curve=&Lerp_Float(A=1, B=1.2, Alpha=$User.ParaGlideLeanAlpha)
     )
     1=&FloatFromCurve(FloatCurve=DataInterface() {propertyValues {Curve {
         defaultValue=0
         keys [
          {interpMode=RCIM_Cubic, tangentMode=RCTM_User, time=0, value=0.75}
          {interpMode=RCIM_Cubic, tangentMode=RCTM_User, time=0.5, value=1}
          {tangentMode=RCTM_User, time=1, value=0.75}
         ]
        }}})
     2=?
     3=?
     Index=&SpawnGroup()
     Num=::NewEnumerator0
    )
   )
   &ScaleColor(
    Scale Alpha=&MergeHelper_FloatSwitch(
     0=1
     1=1
     2=1
     3=&Lerp_Float(A=1, B=1.2, Alpha=$User.ParaGlideLeanAlpha)
     4=&Lerp_Float(A=0.9, B=1.4, Alpha=$User.ParaGlideLeanAlpha)
     Index=&SpawnGroup()
     Num=::NewEnumerator3
    )
    Scale RGB=&MergeHelper_VectorSwitch(
     0=<1 1 1>
     1=&Multiply_Vector(
      A=&MakeVectorFromLinearColorRGB(Color=&ColorFromCurve(
        ColorCurve=DataInterface() {propertyValues {
          AlphaCurve {keys [<at 0 1>]}
          BlueCurve {keys [<at 0 1>,<at 0.4822 0.6622>,<at 1 1>]}
          GreenCurve {keys [<at 0 0.7599>,<at 0.4822 0.35>,<at 1 0.7605>]}
          RedCurve {keys [<at 0 0.2507>,<at 0.4822 1>,<at 1 0.2502>]}
         }}
        CurveIndex=$System.NormalizedLoopAge
       ))
      B=&VectorFromFloat(Value=&Lerp_Float(A=1, B=5, Alpha=$User.ParaGlideLeanAlpha))
     )
     2=&Multiply_VectorByFloat(Vector=<1 1 1>, Float=&Lerp_Float(A=1, B=1.5, Alpha=$User.ParaGlideLeanAlpha))
     3=?
     Index=&SpawnGroup()
     Num=::NewEnumerator1
    )
    ScaleA=true
    ScaleRGB=true
   ) {module=ScaleColor001}
   &ScaleMeshSize(Scale Factor=&VectorFromFloat(Value=&FloatFromCurve(FloatCurve=DataInterface() {propertyValues {Curve {
         defaultValue=1
         keys [{interpMode=RCIM_Cubic, tangentMode=RCTM_User, time=0, value=0}, {interpMode=RCIM_Cubic, tangentMode=RCTM_User, time=0.5, value=1}]
        }}})))
  ]
  rendererClasses [&NiagaraSpriteRendererProperties, &NiagaraMeshRendererProperties]
  renderers [
   {rendererClass=&NiagaraSpriteRendererProperties}
   {rendererClass=&NiagaraSpriteRendererProperties, rendererIndex=1}
   {rendererClass=&NiagaraSpriteRendererProperties, rendererIndex=2}
   {rendererClass=&NiagaraMeshRendererProperties, rendererIndex=3}
   {rendererClass=&NiagaraMeshRendererProperties, rendererIndex=4}
  ]
 }
 Dots_Smoke_FlyAway001 {
  emitterUpdateScript [
   &EmitterState(Life Cycle Mode=::NewEnumerator1, Loop Duration=1, UseLoopDelay=false)
   &SpawnRate(SpawnRate=&Lerp_Float(A=20, B=45, Alpha=$User.ParaGlideLeanAlpha), Use Spawn Probability=false)
   &SpawnRate(SpawnGroup=1, SpawnRate=&Lerp_Float(A=40, B=60, Alpha=$User.ParaGlideLeanAlpha), Use Spawn Probability=false) {module=SpawnRate001}
   &SpawnRate(SpawnGroup=2, SpawnRate=16, Use Spawn Probability=false) {module=SpawnRate002}
  ]
  particleSpawnScript [
   /SetVariables(Particles.VisibilityTag=&SpawnGroup()) {isSetParameters=true}
   &PartitionParticles()
   &InitializeParticle(
    Color=&MergeHelper_ColorSwitch(
     0=&FNTimeOfDayColor(Night Color=<rgba 2.3 2.3 2.3 1>, Morning Color=<rgba 1.8 1.8 1.8 1>, Day Color=<rgba 1.8 1.8 1.8 1>, Evening Color=<rgba 2.3 2.3 2.3 1>)
     1=&MakeCustomLinearColorFromBool(
      Bool=$User.bIsFrontEnd
      True Color=<rgba 0.0243 0.0264 0.065 2>
      False Color=&FNTimeOfDayColor(
       Night Color=<rgba 0.0009 0.0019 0.02 2>
       Morning Color=<rgba 0.0466 0.0316 0.1102 2>
       Day Color=<rgba 0.0259 0.0306 0.1146 2>
       Evening Color=<rgba 0.0009 0.0019 0.02 2>
      )
     )
     2=&FNTimeOfDayColor(Night Color=<rgba 1.8 1.8 1.8 1>, Morning Color=<rgba 1.1042 0.9961 1.5 1>, Day Color=<rgba 1 1 1 1>, Evening Color=<rgba 1.8 1.8 1.8 1>)
     3=?
     Index=&SpawnGroup()
     Num=::NewEnumerator1
    )
    Lifetime=&MergeHelper_FloatSwitch(
     0=&RandomRangeFloat(Minimum=0.5, Maximum=1)
     1=&Lerp_Float(
      A=&RandomRangeFloat(Minimum=0.7, Maximum=1.5)
      B=&RandomRangeFloat(Minimum=0.3, Maximum=0.55)
      Alpha=$User.ParaGlideLeanAlpha
     )
     2=&RandomRangeFloat(Minimum=0.5, Maximum=0.7)
     3=?
     Index=&SpawnGroup()
     Num=::NewEnumerator1
    )
    Sprite Rotation Mode=::NewEnumerator2
    Sprite Size Mode=::NewEnumerator3
    Uniform Sprite Size=&MergeHelper_FloatSwitch(
     0=&RandomRangeFloat(Minimum=2, Maximum=8)
     1=&RandomRangeFloat(Minimum=70, Maximum=90)
     2=&RandomRangeFloat(Minimum=70, Maximum=80)
     3=?
     Index=&SpawnGroup()
     Num=::NewEnumerator1
    )
    UsePositionOffset=false
    Write Material Random=true
   ) {module=InitializeParticle001}
   &SkeletalMeshLocation(
    Bone / Socket Index=$Particles.PartitionParticles.Partition
    Bone Sampling Mode=::NewEnumerator4
    MaskSpawnGroup=false
    Sampled Position Offset=&MergeHelper_VectorSwitch(
     0=&FN_FrontEndVector3Modifier(
      Gameplay=&MergeHelper_VectorSwitch(0=<10 10 23>, 1=<-10 10 23>, 2=?, 3=?, Index=$Particles.PartitionParticles.Partition, Num=::NewEnumerator0)
      Front End=&Lerp_Vector(
       A=&MergeHelper_VectorSwitch(0=<10 -5 23>, 1=<-10 -5 23>, 2=?, 3=?, Index=$Particles.PartitionParticles.Partition, Num=::NewEnumerator0)
       B=&MergeHelper_VectorSwitch(0=<10 -10 23>, 1=<-10 -10 23>, 2=?, 3=?, Index=$Particles.PartitionParticles.Partition, Num=::NewEnumerator0)
       Alpha=$User.ParaGlideLeanAlpha
      )
     )
     1=<0 0 -40>
     2=&MergeHelper_VectorSwitch(0=<10 10 0>, 1=<-10 10 0>, 2=?, 3=?, Index=$Particles.PartitionParticles.Partition, Num=::NewEnumerator0)
     3=?
     Index=&SpawnGroup()
     Num=::NewEnumerator1
    )
    Sampled Position Offset Coordinate Space:ENiagaraSkelMeshTransforms=::NewEnumerator2
    Skeletal Mesh=$System.Skeletal Mesh Reader
   )
   &ShapeLocation(
    MaskSpawnGroup=false
    Offset Mode=::NewEnumerator0
    Override Seed=false
    Rotation Mode=::NewEnumerator1
    Sphere Radius=&MergeHelper_FloatSwitch(0=18, 1=10, 2=5, 3=?, Index=&SpawnGroup(), Num=::NewEnumerator1)
    Sphere Surface Distribution=1
    UseOwnerScale=true
   ) {module=ShapeLocation002}
   &AddVelocity(
    Module.Use Falloff Distance=false
    Origin Offset=<0 0 -100>
    Override Seed=false
    Velocity Mode=::NewEnumerator1
    Velocity Speed=&MergeHelper_FloatSwitch(
     0=&RandomRangeFloat(Minimum=90, Maximum=200)
     1=150
     2=0
     3=?
     Index=&SpawnGroup()
     Num=::NewEnumerator1
    )
   ) {module=AddVelocity001}
   &AddVelocity(Velocity=&MergeHelper_VectorSwitch(
     0=&FN_FrontEndVector3Modifier(
      Gameplay=&Lerp_Vector(
       A=&RandomRangeVector(Minimum=<0 0 10>, Maximum=<0 0 200>)
       B=&RandomRangeVector(Minimum=<0 0 600>, Maximum=<0 0 400>)
       Alpha=$User.ParaGlideLeanAlpha
      )
      Front End=&Lerp_Vector(
       A=&RandomRangeVector(Minimum=<0 -70 10>, Maximum=<0 -30 200>)
       B=&RandomRangeVector(Minimum=<0 -100 250>, Maximum=<0 -130 150>)
       Alpha=$User.ParaGlideLeanAlpha
      )
     )
     1=&FN_FrontEndVector3Modifier(
      Gameplay=&Lerp_Vector(A=<0 0 0>, B=<0 0 200>, Alpha=$User.ParaGlideLeanAlpha)
      Front End=&Lerp_Vector(A=<0 -100 0>, B=<0 -140 200>, Alpha=$User.ParaGlideLeanAlpha)
     )
     2=&Lerp_Vector(A=<0 0 100>, B=<0 0 120>, Alpha=$User.ParaGlideLeanAlpha)
     3=?
     Index=&SpawnGroup()
     Num=::NewEnumerator1
    ))
  ]
  particleUpdateScript [
   &ParticleState()
   &ScaleSpriteSize(
    Scale Sprite Size Mode=::NewEnumerator0
    Uniform Scale Factor=&MergeHelper_FloatSwitch(
     0=&FloatFromCurve(
      FloatCurve=DataInterface() {propertyValues {Curve {
         defaultValue=0
         keys [
          {arriveTangent=4.0731, interpMode=RCIM_Cubic, leaveTangent=4.0731, tangentMode=RCTM_User, time=0, value=0}
          {arriveTangentWeight=2.1629, interpMode=RCIM_Cubic, leaveTangentWeight=0.0333, tangentMode=RCTM_Break, time=0.2, value=3}
          {arriveTangentWeight=0.044, interpMode=RCIM_Cubic, leaveTangentWeight=0.2129, tangentMode=RCTM_User, time=0.3, value=1}
          {arriveTangentWeight=0.3068, interpMode=RCIM_Cubic, leaveTangentWeight=0.1393, tangentMode=RCTM_User, time=0.7335, value=1}
          {interpMode=RCIM_Cubic, tangentMode=RCTM_User, time=1, value=0}
         ]
        }}}
      Scale Curve=&Lerp_Float(A=1, B=1.5, Alpha=$User.ParaGlideLeanAlpha)
     )
     1=&FloatFromCurve(FloatCurve=DataInterface() {propertyValues {Curve {
         defaultValue=0
         keys [{interpMode=RCIM_Cubic, tangentMode=RCTM_User, time=0, value=1}, {interpMode=RCIM_Cubic, tangentMode=RCTM_User, time=1, value=2}]
        }}})
     2=&FloatFromCurve(FloatCurve=DataInterface() {propertyValues {Curve {
         defaultValue=1
         keys [{interpMode=RCIM_Cubic, tangentMode=RCTM_User, time=0, value=1}, {interpMode=RCIM_Cubic, tangentMode=RCTM_User, time=1, value=0}]
        }}})
     3=?
     Index=&SpawnGroup()
     Num=::NewEnumerator1
    )
   )
   &ScaleColor(
    Scale Alpha=&MergeHelper_FloatSwitch(
     0=&FloatFromCurve(FloatCurve=DataInterface() {propertyValues {Curve {keys [<at 0.85 1 RCIM_Cubic>,<at 1 0>]}}})
     1=&FloatFromCurve(
      FloatCurve=DataInterface() {propertyValues {Curve {keys [
          {arriveTangent=0.0539, interpMode=RCIM_Cubic, leaveTangent=0.0539, tangentMode=RCTM_User, time=0, value=0}
          {arriveTangentWeight=0.0338, interpMode=RCIM_Cubic, leaveTangentWeight=0.1974, tangentMode=RCTM_Break, time=0.1, value=1}
          {arriveTangent=-1.1111, interpMode=RCIM_Cubic, leaveTangent=-1.1111, time=0.17, value=1}
          {tangentMode=RCTM_User, time=1, value=0}
         ]}}}
      Scale Curve=&Multiply_Float(
       A=&MakeCustomFloatFromBool(
        Bool=&SetBoolByFloatComparison(
         A=&FN_PickFloatDependOnPlatform(
          TargetPlatform=DataInterface() {propertyValues {Platforms {
             qualityLevelMask=-30
             deviceProfileStates [
              {profileName=Windows, qualityLevelMask=0, setQualityLevelMask=2}
              {profileName=Mobile, qualityLevelMask=0, setQualityLevelMask=4294967295}
              {profileName=Mac, qualityLevelMask=0, setQualityLevelMask=2}
              {profileName=Linux, qualityLevelMask=0, setQualityLevelMask=2}
             ]
            }}}
          TargetPlatformActive=1
          TargetPlatformNoActive=0
         )
         B=0
        )
        True Float=0.5
        False Float=1
       )
       B=&Multiply_Float(
        A=&MakeCustomFloatFromBool(
         Bool=&SetBoolByIntComparison(A=$Engine.QualityLevel, B=0, Comparison Type=::NewEnumerator4)
         True Float=0.65
         False Float=1
        )
        B=&Multiply_Float(
         A=&FN_FrontEndFloatModifier(Gameplay=0.22, Front End=0.12)
         B=&FNTimeOfDayScalar(Night Scale=0.5, Morning Scale=0.75, Evening Scale=0.75)
        )
       )
      )
     )
     2=&FloatFromCurve(
      FloatCurve=DataInterface() {propertyValues {Curve {
         defaultValue=1
         keys [
          {interpMode=RCIM_Cubic, tangentMode=RCTM_User, time=0, value=0.5}
          {interpMode=RCIM_Cubic, tangentMode=RCTM_User, time=0.2, value=1}
          {tangentMode=RCTM_User, time=1, value=0}
         ]
        }}}
      Scale Curve=&RandomRangeFloat(Minimum=0.08, Maximum=0.12)
     )
     3=?
     Index=&SpawnGroup()
     Num=::NewEnumerator1
    )
    Scale RGB=&MergeHelper_VectorSwitch(
     0=&MakeVectorFromLinearColorRGB(Color=&ColorFromCurve(ColorCurve=DataInterface() {propertyValues {
         AlphaCurve {keys [<at 0.5433 1 RCIM_Cubic>,<at 1 0 RCIM_Cubic>]}
         BlueCurve {keys [{interpMode=RCIM_Cubic, tangentMode=RCTM_User, time=0.0619, value=0.9613}, <at 0.8868 3 RCIM_Cubic>]}
         GreenCurve {keys [{interpMode=RCIM_Cubic, tangentMode=RCTM_User, time=0.0619, value=0.3}, <at 0.8868 2.7449 RCIM_Cubic>]}
         RedCurve {keys [{interpMode=RCIM_Cubic, tangentMode=RCTM_User, time=0.0619, value=1}, <at 0.8868 0.2865 RCIM_Cubic>]}
        }}))
     1=&VectorFromCurve(VectorCurve=DataInterface() {propertyValues {
        XCurve {
         defaultValue=1
         keys [{interpMode=RCIM_Cubic, tangentMode=RCTM_User, time=0, value=1}, {interpMode=RCIM_Cubic, tangentMode=RCTM_User, time=1, value=0.3}]
        }
        YCurve {
         defaultValue=1
         keys [{interpMode=RCIM_Cubic, tangentMode=RCTM_User, time=0, value=1}, {interpMode=RCIM_Cubic, tangentMode=RCTM_User, time=1, value=0.3}]
        }
        ZCurve {
         defaultValue=1
         keys [{interpMode=RCIM_Cubic, tangentMode=RCTM_User, time=0, value=1}, {interpMode=RCIM_Cubic, tangentMode=RCTM_User, time=1, value=0.3}]
        }
       }})
     2=&VectorFromCurve(VectorCurve=DataInterface() {propertyValues {
        XCurve {
         defaultValue=1
         keys [{interpMode=RCIM_Cubic, tangentMode=RCTM_User, time=0, value=1}, {interpMode=RCIM_Cubic, tangentMode=RCTM_User, time=1, value=0.3}]
        }
        YCurve {
         defaultValue=1
         keys [{interpMode=RCIM_Cubic, tangentMode=RCTM_User, time=0, value=1}, {interpMode=RCIM_Cubic, tangentMode=RCTM_User, time=1, value=0.3}]
        }
        ZCurve {
         defaultValue=1
         keys [{interpMode=RCIM_Cubic, tangentMode=RCTM_User, time=0, value=1}, {interpMode=RCIM_Cubic, tangentMode=RCTM_User, time=1, value=0.3}]
        }
       }})
     3=?
     Index=&SpawnGroup()
     Num=::NewEnumerator1
    )
    ScaleA=true
    ScaleRGB=true
   ) {module=ScaleColor001}
   &SubUVAnimation(
    End Frame Range Override=40
    Output Alternate Renderer Binding=true
    "Sprite Renderer "=DataInterface() {propertyValues {SpriteRenderer=&self:Dots_Smoke_FlyAway001_0.NiagaraSpriteRendererProperties_1}}
    SubUV Lookup Index Scale=0.6
    UseEndFrame=true
    UseStartFrame=true
   ) {module=SubUVAnimation001}
   &CurlNoiseForce(
    Bias Noise=false
    Noise Frequency=&MaskFloat(Mask Source=0, Mask Parameter=?, Mask=0, Masked Float=50, Passthrough Float=1)
    Noise Strength=&MaskFloat(Mask Source=0, Mask Parameter=?, Mask=0, Masked Float=150, Passthrough Float=1)
    Pan Noise=false
   )
   &SolveForcesAndVelocity(Clamp Velocity=false, Limit Acceleration=false)
  ]
  rendererClasses [&NiagaraSpriteRendererProperties]
  renderers [
   {rendererClass=&NiagaraSpriteRendererProperties}
   {rendererClass=&NiagaraSpriteRendererProperties, rendererIndex=1}
   {rendererClass=&NiagaraSpriteRendererProperties, rendererIndex=2}
  ]
 }
 SmokeGalaxy_Hands001 {
  emitterUpdateScript [
   &EmitterState(Life Cycle Mode=::NewEnumerator1, Loop Duration=1, UseLoopDelay=false)
   &SpawnRate(SpawnRate=15, Use Spawn Probability=false)
   &SpawnRate(SpawnGroup=1, SpawnRate=10, Use Spawn Probability=false) {module=SpawnRate001}
   &SpawnRate(SpawnGroup=2, SpawnRate=2, Use Spawn Probability=false) {module=SpawnRate002}
  ]
  particleSpawnScript [
   /SetVariables(Particles.VisibilityTag=&SpawnGroup()) {isSetParameters=true}
   &PartitionParticles()
   &InitializeParticle(
    Color=&MergeHelper_ColorSwitch(
     0=&MakeCustomLinearColorFromBool(
      Bool=$User.bIsFrontEnd
      True Color=<rgba 0.0168 0.0184 0.0451 1.1>
      False Color=&FNTimeOfDayColor(
       Night Color=<rgba 0.0062 0.0075 0.03 1.1>
       Morning Color=<rgba 0.0196 0.0113 0.03 1.1>
       Day Color=<rgba 0.0112 0.0122 0.03 1.1>
       Evening Color=<rgba 0.0062 0.0075 0.03 1.1>
      )
     )
     1=&MakeCustomLinearColorFromBool(
      Bool=$User.bIsFrontEnd
      True Color=<rgba 0.0336 0.0366 0.09 1.2>
      False Color=&FNTimeOfDayColor(
       Night Color=<rgba 0.0112 0.0122 0.03 1.2>
       Morning Color=<rgba 0.045 0.0221 0.06 1.2>
       Day Color=<rgba 0.0224 0.0244 0.06 1.2>
       Evening Color=<rgba 0.0112 0.0122 0.03 1.2>
      )
     )
     2=&FNTimeOfDayColor(
      Night Color=<rgba 0.2612 0.8486 1 0.25>
      Morning Color=<rgba 0.5336 0.4856 1 0.25>
      Day Color=<rgba 0.2612 0.8486 1 0.25>
      Evening Color=<rgba 0.2612 0.8486 1 0.25>
     )
     3=?
     Index=&SpawnGroup()
     Num=::NewEnumerator1
    )
    Lifetime=&MergeHelper_FloatSwitch(
     0=&RandomRangeFloat(Minimum=2, Maximum=4)
     1=&RandomRangeFloat(Minimum=2, Maximum=4)
     2=1.5
     3=?
     Index=&SpawnGroup()
     Num=::NewEnumerator1
    )
    Sprite Rotation Mode=::NewEnumerator2
    Sprite Size Mode=::NewEnumerator3
    Uniform Sprite Size=&MergeHelper_FloatSwitch(
     0=&RandomRangeFloat(Minimum=60, Maximum=80)
     1=&Lerp_Float(
      A=&RandomRangeFloat(Minimum=150, Maximum=200)
      B=&RandomRangeFloat(Minimum=160, Maximum=210)
      Alpha=$User.ParaGlideLeanAlpha
     )
     2=50
     3=?
     Index=&SpawnGroup()
     Num=::NewEnumerator1
    )
    UsePositionOffset=false
    Write Material Random=true
   )
  ]
  particleUpdateScript [
   &SkeletalMeshLocation(
    Bone / Socket Index=$Particles.PartitionParticles.Partition
    Bone Sampling Mode=::NewEnumerator4
    MaskSpawnGroup=false
    Sampled Position Offset=&MergeHelper_VectorSwitch(
     0=&MergeHelper_VectorSwitch(0=<10 10 0>, 1=<-10 -10 0>, 2=?, 3=?, Index=$Particles.PartitionParticles.Partition, Num=::NewEnumerator0)
     1=&MergeHelper_VectorSwitch(0=<10 10 0>, 1=<-10 -10 0>, 2=?, 3=?, Index=$Particles.PartitionParticles.Partition, Num=::NewEnumerator0)
     2=&MergeHelper_VectorSwitch(0=<0 10 0>, 1=<0 -10 0>, 2=?, 3=?, Index=$Particles.PartitionParticles.Partition, Num=::NewEnumerator0)
     3=?
     Index=&SpawnGroup()
     Num=::NewEnumerator1
    )
    Skeletal Mesh=$System.Skeletal Mesh Reader
   )
   &ParticleState()
   &SubUVAnimation(
    End Frame Range Override=40
    Output Alternate Renderer Binding=true
    "Sprite Renderer "=DataInterface() {propertyValues {SpriteRenderer=&self:SmokeGalaxy_Hands001_0.NiagaraSpriteRendererProperties_1}}
    SubUV Lookup Index Scale=0.6
    UseEndFrame=true
    UseStartFrame=true
   ) {module=SubUVAnimation001}
   &ScaleColor(
    Scale Alpha=&MergeHelper_FloatSwitch(
     0=&FloatFromCurve(
      FloatCurve=DataInterface() {propertyValues {Curve {keys [
          {interpMode=RCIM_Cubic, tangentMode=RCTM_User, time=0, value=0}
          {interpMode=RCIM_Cubic, tangentMode=RCTM_User, time=0.2766, value=1}
          {tangentMode=RCTM_User, time=1, value=0}
         ]}}}
      Scale Curve=&Multiply_Float(
       A=&MakeCustomFloatFromBool(
        Bool=&SetBoolByFloatComparison(
         A=&FN_PickFloatDependOnPlatform(
          TargetPlatform=DataInterface() {propertyValues {Platforms {
             qualityLevelMask=-30
             deviceProfileStates [
              {profileName=Windows, qualityLevelMask=0, setQualityLevelMask=2}
              {profileName=Mobile, qualityLevelMask=0, setQualityLevelMask=4294967295}
              {profileName=Mac, qualityLevelMask=0, setQualityLevelMask=2}
              {profileName=Linux, qualityLevelMask=0, setQualityLevelMask=2}
             ]
            }}}
          TargetPlatformActive=1
          TargetPlatformNoActive=0
         )
         B=0
        )
        True Float=0.5
        False Float=1
       )
       B=&Multiply_Float(
        A=&MakeCustomFloatFromBool(
         Bool=&SetBoolByIntComparison(A=$Engine.QualityLevel, B=0, Comparison Type=::NewEnumerator4)
         True Float=0.65
         False Float=1
        )
        B=&Multiply_Float(
         A=&FN_FrontEndFloatModifier(Gameplay=0.22, Front End=0.12)
         B=&FNTimeOfDayScalar(Night Scale=0.5, Morning Scale=0.75, Evening Scale=0.75)
        )
       )
      )
     )
     1=&FloatFromCurve(
      FloatCurve=DataInterface() {propertyValues {Curve {keys [
          {interpMode=RCIM_Cubic, tangentMode=RCTM_User, time=0, value=0}
          {interpMode=RCIM_Cubic, tangentMode=RCTM_User, time=0.2766, value=1}
          {tangentMode=RCTM_User, time=1, value=0}
         ]}}}
      Scale Curve=&Multiply_Float(
       A=&MakeCustomFloatFromBool(
        Bool=&SetBoolByFloatComparison(
         A=&FN_PickFloatDependOnPlatform(
          TargetPlatform=DataInterface() {propertyValues {Platforms {
             qualityLevelMask=-30
             deviceProfileStates [
              {profileName=Windows, qualityLevelMask=0, setQualityLevelMask=2}
              {profileName=Mobile, qualityLevelMask=0, setQualityLevelMask=4294967295}
              {profileName=Mac, qualityLevelMask=0, setQualityLevelMask=2}
              {profileName=Linux, qualityLevelMask=0, setQualityLevelMask=2}
             ]
            }}}
          TargetPlatformActive=1
          TargetPlatformNoActive=0
         )
         B=0
        )
        True Float=0.5
        False Float=1
       )
       B=&Multiply_Float(
        A=&MakeCustomFloatFromBool(
         Bool=&SetBoolByIntComparison(A=$Engine.QualityLevel, B=0, Comparison Type=::NewEnumerator4)
         True Float=0.65
         False Float=1
        )
        B=&Multiply_Float(
         A=&FN_FrontEndFloatModifier(Gameplay=0.22, Front End=0.12)
         B=&FNTimeOfDayScalar(Night Scale=0.5, Morning Scale=0.75, Evening Scale=0.75)
        )
       )
      )
     )
     2=&FloatFromCurve(FloatCurve=DataInterface() {propertyValues {Curve {
         defaultValue=1
         keys [
          {interpMode=RCIM_Cubic, tangentMode=RCTM_User, time=0, value=0}
          {interpMode=RCIM_Cubic, tangentMode=RCTM_User, time=0.5, value=1}
          {tangentMode=RCTM_User, time=1, value=0}
         ]
        }}})
     3=?
     Index=&SpawnGroup()
     Num=::NewEnumerator1
    )
    Scale RGB=&MergeHelper_VectorSwitch(
     0=&Multiply_Vector(
      A=&VectorFromCurve(
       VectorCurve=DataInterface() {propertyValues {
         XCurve {
          defaultValue=1
          keys [{interpMode=RCIM_Cubic, tangentMode=RCTM_User, time=0, value=1}, {interpMode=RCIM_Cubic, tangentMode=RCTM_User, time=1, value=0}]
         }
         YCurve {
          defaultValue=1
          keys [{interpMode=RCIM_Cubic, tangentMode=RCTM_User, time=0, value=1}, {interpMode=RCIM_Cubic, tangentMode=RCTM_User, time=1, value=0}]
         }
         ZCurve {
          defaultValue=1
          keys [{interpMode=RCIM_Cubic, tangentMode=RCTM_User, time=0, value=1}, {interpMode=RCIM_Cubic, tangentMode=RCTM_User, time=1, value=0}]
         }
        }}
       Scale Curve=&VectorFromFloat(Value=&Lerp_Float(A=1, B=2, Alpha=$User.ParaGlideLeanAlpha))
      )
      B=<1 1 1>
     )
     1=&VectorFromCurve(VectorCurve=DataInterface() {propertyValues {
        XCurve {
         defaultValue=1
         keys [{interpMode=RCIM_Cubic, tangentMode=RCTM_User, time=0, value=1}, {interpMode=RCIM_Cubic, tangentMode=RCTM_User, time=1, value=0}]
        }
        YCurve {
         defaultValue=1
         keys [{interpMode=RCIM_Cubic, tangentMode=RCTM_User, time=0, value=1}, {interpMode=RCIM_Cubic, tangentMode=RCTM_User, time=1, value=0}]
        }
        ZCurve {
         defaultValue=1
         keys [{interpMode=RCIM_Cubic, tangentMode=RCTM_User, time=0, value=1}, {interpMode=RCIM_Cubic, tangentMode=RCTM_User, time=1, value=0}]
        }
       }})
     2=<1 1 1>
     3=?
     Index=&SpawnGroup()
     Num=::NewEnumerator1
    )
    ScaleA=true
    ScaleRGB=true
   )
  ]
  rendererClasses [&NiagaraSpriteRendererProperties]
  renderers [
   {rendererClass=&NiagaraSpriteRendererProperties}
   {rendererClass=&NiagaraSpriteRendererProperties, rendererIndex=1}
   {rendererClass=&NiagaraSpriteRendererProperties, rendererIndex=2}
  ]
 }
 Swirls001 {
  emitterUpdateScript [
   &EmitterState(Life Cycle Mode=::NewEnumerator1, Loop Duration=1, UseLoopDelay=false)
   &SpawnRate(SpawnGroup=1, SpawnRate=&Lerp_Float(A=10, B=25, Alpha=$User.ParaGlideLeanAlpha), Use Spawn Probability=false) {module=SpawnRate001}
  ]
  particleSpawnScript [
   &PartitionParticles()
   &InitializeParticle(
    Color Maximum=&FNTimeOfDayColor(
     Night Color=<rgba 0 1.2553 4 1.5>
     Morning Color=<rgba 0 0.7846 2.5 0.7826>
     Day Color=<rgba 0 0.6277 2 0.7826>
     Evening Color=<rgba 0 0.6277 2 0.7826>
    )
    Color Minimum=&FNTimeOfDayColor(
     Night Color=<rgba 3.4335 0.5729 4 0.6783>
     Morning Color=<rgba 3.445 0.6424 4 0.6783>
     Day Color=<rgba 3.56 1.3382 4 0.6783>
     Evening Color=<rgba 3.4335 0.5729 4 0.6783>
    )
    Color Mode=::NewEnumerator2
    Lifetime=&Lerp_Float(
     A=&RandomRangeFloat(Minimum=0.85, Maximum=1.2)
     B=&RandomRangeFloat(Minimum=0.5, Maximum=0.85)
     Alpha=$User.ParaGlideLeanAlpha
    )
    Mesh Scale=&RandomRangeVector(Minimum=<7 1.5 0.7>, Maximum=<3 2 1.5>)
    Mesh Scale Mode=::NewEnumerator4
    UsePositionOffset=false
    Write Material Random=true
   )
   &CameraOffset(Camera Offset Amount=3)
   &InitialMeshOrientation(
    Mesh Axis To Orient=<0 0 -1>
    Mesh Orientation Mode=::NewEnumerator2
    Orientation Vector=<0 0 1>
    Rotation=&FN_FrontEndVector3Modifier(
     Gameplay=&RandomRangeVector(Minimum=<0 0 0>, Maximum=<0 0 1>)
     Front End=&Lerp_Vector(
      A=&RandomRangeVector(Minimum=<0 -0.07 0>, Maximum=<0 -0.07 1>)
      B=&RandomRangeVector(Minimum=<0 -0.1 0>, Maximum=<0 -0.1 1>)
      Alpha=$User.ParaGlideLeanAlpha
     )
    )
    Rotation Coordinate Space=::NewEnumerator2
    Use Rotation Vector=true
   )
  ]
  particleUpdateScript [
   &ParticleState()
   &SkeletalMeshLocation(
    Bone / Socket Index=$Particles.PartitionParticles.Partition
    Bone Sampling Mode=::NewEnumerator4
    MaskSpawnGroup=false
    Orientation Sampling=::NewEnumerator1
    Sampled Position Offset=&MergeHelper_VectorSwitch(0=<10 10 0>, 1=<-10 -10 0>, 2=?, 3=?, Index=$Particles.PartitionParticles.Partition, Num=::NewEnumerator0)
    Skeletal Mesh=$System.Skeletal Mesh Reader
   )
   &DynamicMaterialParameters(
    Index 0 Param 1=&RandomRangeFloat(Minimum=1, Maximum=2)
    Index 0 Param 2=&Multiply_Float(
     A=&RandomRangeFloat(Minimum=0.7, Maximum=1.5)
     B=&Lerp_Float(A=0.1, B=&FN_FrontEndFloatModifier(Gameplay=2.5, Front End=1.5), Alpha=$System.Dive)
    )
    Index 0 Param 3=&Multiply_Float(
     A=&RandomRangeFloat(Minimum=6, Maximum=3)
     B=&Lerp_Float(A=1, B=&FN_FrontEndFloatModifier(Gameplay=2.5, Front End=1.5), Alpha=$System.Dive)
    )
    Index 0 Param 4=&RandomRangeFloat(Minimum=1, Maximum=1.6)
    Index 1 Param 4=0.7
    Index 1 Vector 1:Vector3f=$Particles.InterpolateOverTime.InterpolateOverTime_V3_.Moving Average
    Index 2 Param 1=&FloatFromCurve(FloatCurve=DataInterface() {propertyValues {Curve {keys [<at 0 1 RCIM_Cubic>,<at 0.4739 1 RCIM_Cubic>,<at 1 1>]}}})
    Index 2 Param 2=&FloatFromCurve(FloatCurve=DataInterface() {propertyValues {Curve {keys [<at 0 -1>,<at 1 1>]}}}, Scale Curve=-1)
    Index 2 Param 3=&RandomRangeFloat(Minimum=10, Maximum=20)
    Index 2 Param 4=1
    Module.Index0Float0Write=true
    Module.Index0Float1Write=true
    Module.Index0Float2Write=true
    Module.Index0Float3Write=true
    Module.Index1Float3Write=true
    Module.Index1VectorWrite=true
    Module.Index2Float0Write=true
    Module.Index2Float1Write=true
    Module.Index2Float2Write=true
    Module.Index2Float3Write=false
    Write Parameter Index 1=::NewEnumerator3
    Write Parameter Index 2=::NewEnumerator1
   )
   &ScaleMeshSize(Scale Factor=&VectorFromFloat(Value=&FloatFromCurve(
      FloatCurve=DataInterface() {propertyValues {Curve {keys [{arriveTangent=1.0186, interpMode=RCIM_Cubic, leaveTangent=1.0186, tangentMode=RCTM_User, time=0, value=0}, <at 1 1 RCIM_Cubic>]}}}
      CurveIndex=$Emitter.Age
      Scale Curve=2
     )))
   &ScaleColor(Color Scale:LinearColor=$System.TOD, Scale Mode=3)
  ]
  rendererClasses [&NiagaraMeshRendererProperties]
  renderers [{rendererClass=&NiagaraMeshRendererProperties, rendererIndex=0}]
 }
 Trail_Hand {
  emitterUpdateScript [
   &EmitterState(
    Inactive Response=::NewEnumerator1
    Life Cycle Mode=::NewEnumerator1
    Loop Behavior=::NewEnumerator1
    Loop Duration Mode=::NewEnumerator1
    UseLoopDelay=false
   )
   &SpawnBurst_Instantaneous(Spawn Count=2, Use Spawn Probability=false, UseLoopCountLimit=false)
   /SetVariables(Emitter.TrailTex_001SpeedU:NiagaraFloat=&MakeCustomFloatFromBool(Bool=&SetBoolByFloatComparison(A=$User.ParaGlideLeanAlpha, B=0.2), True Float=-2.5, False Float=-1.25)) {isSetParameters=true}
  ]
  particleSpawnScript [
   &PartitionParticles()
   &InitializeParticle(
    Color=&FNTimeOfDayColor(
     Night Color=<rgba 0.0046 0.034 0.0712 7>
     Morning Color=<rgba 0.026 0.0574 0.0972 5>
     Day Color=<rgba 0.026 0.0574 0.0972 5>
     Evening Color=<rgba 0.0046 0.034 0.0712 7>
    )
    Mesh Scale Mode=::NewEnumerator3
    Mesh Uniform Scale=&MergeHelper_FloatSwitch(
     0=&MakeFloatFromVector(
      Vector=&SelectVectorFromArray(
       Direct Array Index=$Particles.PartitionParticles.Partition
       Vector Selection Array=DataInterface() {propertyValues {FloatData [<1 1 1>, <1 1 1>]}}
      )
      Channel=::NewEnumerator0
     )
     1=0.13
     2=?
     3=?
     Index=&SpawnGroup()
     Num=::NewEnumerator0
    )
    UsePositionOffset=false
    Write Material Random=true
   )
   &InitialMeshOrientation(
    Mesh Axis To Orient=&Lerp_Vector(
     A=&MergeHelper_VectorSwitch(0=<0.5 0.1 0.03>, 1=<0.5 0.1 -0.03>, 2=?, 3=?, Index=$Particles.PartitionParticles.Partition, Num=::NewEnumerator0)
     B=<0.3 0 0>
     Alpha=$User.ParaGlideLeanAlpha
    )
    Mesh Orientation Mode=::NewEnumerator2
    Orientation Coordinate Space=::Local
    Rotation=<0 -0.25 0>
    Rotation Coordinate Space=::NewEnumerator2
    Use Rotation Vector=true
   ) {module=InitialMeshOrientation001}
  ]
  particleUpdateScript [
   &ParticleState(Kill Particles When Lifetime Has Elapsed=false)
   &SkeletalMeshLocation(
    Bone / Socket Index=$Particles.PartitionParticles.Partition
    Bone Sampling Mode=::NewEnumerator4
    MaskSpawnGroup=false
    Orientation Sampling=::NewEnumerator1
    Sampled Position Offset=&MergeHelper_VectorSwitch(
     0=&SelectVectorFromArray(
      Direct Array Index=$Particles.PartitionParticles.Partition
      Vector Selection Array=DataInterface() {propertyValues {FloatData [<0 0 -20>, <0 0 -20>]}}
     )
     1=&MergeHelper_VectorSwitch(0=<10 10 0>, 1=<-10 -10 0>, 2=?, 3=?, Index=$Particles.PartitionParticles.Partition, Num=::NewEnumerator0)
     2=?
     3=?
     Index=&SpawnGroup()
     Num=::NewEnumerator0
    )
    Sampled Position Offset Coordinate Space:ENiagaraSkelMeshTransforms=::NewEnumerator2
    Skeletal Mesh=$System.Skeletal Mesh Reader
   ) {module=SkeletalMeshLocation001}
   &DynamicMaterialParameters(
    Index 2 Param 1=1
    Index 2 Param 3=&MakeCustomFloatFromBool(Bool=$User.bisFrontEndPreview, True Float=0.7, False Float=1)
    Module.Index2Float0Write=true
    Module.Index2Float1Write=true
    Module.Index2Float2Write=true
    Module.Index2Float3Write=false
    Write Parameter Index 0=::NewEnumerator0
    Write Parameter Index 2=::NewEnumerator1
   )
  ]
  rendererClasses [&NiagaraMeshRendererProperties]
  renderers [{rendererClass=&NiagaraMeshRendererProperties, rendererIndex=0}]
 }
 Vortex_Both001 {
  emitterUpdateScript [
   &EmitterState()
   &SpawnRate(SpawnRate=&Lerp_Float(A=75, B=90, Alpha=$User.ParaGlideLeanAlpha), Use Spawn Probability=false)
  ]
  particleSpawnScript [
   &InitializeParticle(
    Color=&FNTimeOfDayColor(Night Color=<rgba 5 5 5 1>, Morning Color=<rgba 1 1 1 1>, Day Color=<rgba 1 1 1 1>, Evening Color=<rgba 2.5 2.5 2.5 1>)
    Lifetime=0.7
    Sprite Size=&Lerp_Vector2(
     A=&RandomRangeVector2D(Minimum=<1 3>, Maximum=<2 5>)
     B=&RandomRangeVector2D(Minimum=<1.5 3.5>, Maximum=<3 7>)
     Alpha=$User.ParaGlideLeanAlpha
    )
    Sprite Size Mode=::NewEnumerator4
    UsePositionOffset=false
    Write Material Random=true
   )
   &CameraOffset(Camera Offset Amount=$System.CameraOffset)
   &PartitionParticles()
  ]
  particleUpdateScript [
   &ParticleState()
   &SkeletalMeshLocation(
    Bone / Socket Index=$Particles.PartitionParticles.Partition
    Bone Sampling Mode=::NewEnumerator4
    MaskSpawnGroup=false
    Position Sampling=::NewEnumerator1
    Sampled Position Offset=&MergeHelper_VectorSwitch(0=<10 10 0>, 1=<-10 -10 0>, 2=?, 3=?, Index=$Particles.PartitionParticles.Partition, Num=::NewEnumerator0)
    Skeletal Mesh=$System.Skeletal Mesh Reader
   ) {module=SkeletalMeshLocation001}
   /SetVariables(Particles.LeftQuat:Quat4f=&MultiplyQuaternion(
     Quaternion A=$Particles.SkeletalMeshLocation001.SampledRotation
     Quaternion B=&MakeQuaternion(
      Quaternion From=0
      Axis=$System.OrientationLeft
      Angle=&FN_FrontEndFloatModifier(Gameplay=110, Front End=160)
      From=?
      To=?
      Stabilization Mode:ENiagara_OrientVectorCount=::NewEnumerator0
      Side Direction:Vector3f=?
      Up Direction:Vector3f=?
      Angle Type=::NewEnumerator0
      Coordinate Space=::Local
      Normalize:NiagaraBool=true
     )
    )) {isSetParameters=true}
   &ShapeLocation(
    Additional Quaternion Rotation=false
    Handle Radius=3
    Large Radius=&Lerp_Float(A=40, B=55, Alpha=$User.ParaGlideLeanAlpha)
    MaskSpawnGroup=false
    Non Uniform Scale=<1 1 0.3>
    Offset Coordinate Space=::Local
    Offset Mode=::NewEnumerator0
    Override Seed=false
    Rotation Mode=::NewEnumerator3
    Rotation Quaternion=$Particles.LeftQuat
    Shape Origin=$Particles.SkeletalMeshLocation001.SampledPosition
    Shape Primitive=::NewEnumerator3
    UseOwnerScale=true
   )
   &PointAttractionForce(
    Attraction Radius=50
    Attraction Strength=&Lerp_Float(A=90, B=250, Alpha=$User.ParaGlideLeanAlpha)
    Attractor Position:NiagaraPosition=$Particles.SkeletalMeshLocation001.SampledPosition
    Attractor Position Offset Coordinate Space:ENiagaraCoordinateSpace=::Local
    Kill Within Radius=false
    Module.Use Falloff=true
    UseAttractorOffset=true
   )
   &VortexForce(
    Origin Pull Amount:NiagaraFloat=&Lerp_Float(A=500, B=1200, Alpha=$User.ParaGlideLeanAlpha)
    UseOriginOffset=false
    Vortex Axis=&RotateVectorByQuaternion(Vector To Rotate=<0 0 1>, Quaternion=$Particles.LeftQuat)
    Vortex Axis Coordinate Space:ENiagaraCoordinateSpace=::Local
    Vortex Force Amount=&Lerp_Float(A=700, B=1500, Alpha=$User.ParaGlideLeanAlpha)
    Vortex Origin:NiagaraPosition=$Particles.SkeletalMeshLocation001.SampledPosition
   )
   &ScaleSpriteSize(Uniform Curve Sprite Scale=DataInterface() {propertyValues {Curve {
       defaultValue=1
       keys [
        {interpMode=RCIM_Cubic, tangentMode=RCTM_User, time=0, value=0}
        {interpMode=RCIM_Cubic, tangentMode=RCTM_User, time=0.5, value=1}
        {tangentMode=RCTM_User, time=0.9929, value=0.3824}
       ]
      }}})
   &SolveForcesAndVelocity(Clamp Velocity=false, Limit Acceleration=false)
   &ScaleColor(
    Scale RGB=&MakeVectorFromLinearColorRGB(Color=&ColorFromCurve(ColorCurve=DataInterface() {propertyValues {
        AlphaCurve {keys [<at 0 1 RCIM_Cubic>,<at 1 0 RCIM_Cubic>]}
        BlueCurve {keys [{interpMode=RCIM_Cubic, tangentMode=RCTM_User, time=0.2674, value=0.9613}, <at 1 3 RCIM_Cubic>]}
        GreenCurve {keys [{interpMode=RCIM_Cubic, tangentMode=RCTM_User, time=0.2674, value=0.3}, <at 1 2.7449 RCIM_Cubic>]}
        RedCurve {keys [{interpMode=RCIM_Cubic, tangentMode=RCTM_User, time=0.2674, value=1}, <at 1 0.2865 RCIM_Cubic>]}
       }}))
    ScaleA=true
    ScaleRGB=true
   )
  ]
  rendererClasses [&NiagaraSpriteRendererProperties]
  renderers [{rendererClass=&NiagaraSpriteRendererProperties}]
 }
}
summary {
 systemName=NS_Contrail_DistanceHappening
 emitters [
  {emitterName=Vortex_Both001, rendererClasses [&NiagaraSpriteRendererProperties]}
  {emitterName=Swirls001, rendererClasses [&NiagaraMeshRendererProperties]}
  {emitterName=SmokeGalaxy_Hands001, rendererClasses [&NiagaraSpriteRendererProperties]}
  {emitterName=Dots_Smoke_FlyAway001, rendererClasses [&NiagaraSpriteRendererProperties]}
  {emitterName=Trail_Hand, rendererClasses [&NiagaraMeshRendererProperties]}
  {
   emitterName=Core_Glow_Refraction
   rendererClasses [&NiagaraSpriteRendererProperties, &NiagaraMeshRendererProperties]
  }
 ]
 userVariables [
  User.bIsFrontEnd {type=NiagaraBool}
  User.bIsFrontEndPreview {type=NiagaraBool}
  User.ParaGlideLeanAlpha {type=NiagaraFloat, defaultValue=0, description=""}
  User.RibbonIDOffset {type=NiagaraInt32}
  User.VortexLeft {type=Vector3f, description="", defaultValue {struct=Vector3f, value=<-35 50 90>}}
  User.VortexRight {type=Vector3f, description="", defaultValue {struct=Vector3f, value=<35 0 90>}}
 ]
}
